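# Sound cue for the "B" step sound of the Athena HappySkipping emote.
# A Random node picks one of three equally weighted wave variations, then a Modulator applies a very slight pitch variation.
# It plays through emote attenuation, the emote music sound class and a shared emote music concurrency set, with boosted volume.
# It also sends to a music audio-analysis submix.
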SoundCue /Game/Athena/Sounds/Emotes/HappySkipping/Athena_Emote_HappySkipping_Step_B_Cue
root=Modulator_0
nodes [
 {id=Modulator_0, children [Random_0], properties {pitchMax=1.01, pitchMin=0.99, volumeMax=1}}
 {id=Random_0, children [WavePlayer_1,WavePlayer_2,WavePlayer_3], properties {weights [1,1,1]}}
 {id=WavePlayer_1, properties {soundWaveAssetPtr=&/Athena_Emote_HappySkipping_B_01}}
 {id=WavePlayer_2, properties {soundWaveAssetPtr=&/Athena_Emote_HappySkipping_B_02}}
 {id=WavePlayer_3, properties {soundWaveAssetPtr=&/Athena_Emote_HappySkipping_B_03}}
]
properties {
 attenuationSettings=&Fort_Emotes_Athena_Att
 maxDistance=1600
 soundClassObject=&Fort_Music_Emote
 volumeMultiplier=1.5
 concurrencyOverrides {resolutionRule=PreventNew}
 concurrencySet [/Game/Sounds/Concurrency_Rules/Athena_EmoteMusic_Conc]
 soundSubmixSends [{minSendLevel=0, soundSubmix=&Fortnite_AudioAnalysis_Music}]
}
